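# Sound cue for the final impact when a small ceramic object is demolished.
# It picks one of two ceramic impact waves at equal weight and varies pitch through a Modulator, on top of a raised cue pitch and reduced volume.
# It uses medium-object destruction attenuation, the impacts sound class and a shared four-max object-impact concurrency set.
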SoundCue /Game/Sounds/Fort_Impact_Sounds/Demolition/Ceramic/Demo_Ceramic_Small_ImpactFinal_Cue
root=Modulator_2
nodes [
 {id=Modulator_2, children [Random_0], properties {pitchMax=1.07, pitchMin=0.92}}
 {id=Random_0, children [WavePlayer_0,WavePlayer_3], properties {weights [1,1]}}
 {id=WavePlayer_0, properties {soundWaveAssetPtr=&/Ceramic_ImpactFinal01}}
 {id=WavePlayer_3, properties {soundWaveAssetPtr=&/Ceramic_ImpactFinal04}}
]
properties {
 attenuationSettings=/Game/Sounds/Attenuation/Fort_Demolition_Destroy_Medium_Sized_Object_Att
 maxDistance=4000
 pitchMultiplier=1.6
 soundClassObject=&Fort_Impacts_PSM
 volumeMultiplier=0.544
 concurrencyOverrides {resolutionRule=PreventNew}
 concurrencySet [&CONC_Object_Impacts_FourMax]
}
