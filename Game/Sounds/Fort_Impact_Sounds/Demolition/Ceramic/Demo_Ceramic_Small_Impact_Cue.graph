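# Sound cue for small ceramic impacts in the demolition impact sound set.
# A random node picks one of four equally weighted ceramic impact waves, and a modulator adds slight pitch variation.
# The cue plays pitched well up at reduced volume in the impacts sound class, with small-object demolition hit attenuation.
# An object-impacts concurrency set is assigned, alongside override settings that stop the oldest sound.
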SoundCue /Game/Sounds/Fort_Impact_Sounds/Demolition/Ceramic/Demo_Ceramic_Small_Impact_Cue
root=Modulator_2
nodes [
 {id=Modulator_2, children [Random_1], properties {pitchMax=1.07, pitchMin=0.92}}
 {id=Random_1, children [WavePlayer_4,WavePlayer_5,WavePlayer_6,WavePlayer_7], properties {weights [1,1,1,1]}}
 {id=WavePlayer_4, properties {soundWaveAssetPtr=&/Ceramic_Impact01}}
 {id=WavePlayer_5, properties {soundWaveAssetPtr=&/Ceramic_Impact02}}
 {id=WavePlayer_6, properties {soundWaveAssetPtr=&/Ceramic_Impact03}}
 {id=WavePlayer_7, properties {soundWaveAssetPtr=&/Ceramic_Impact04}}
]
properties {
 attenuationSettings=/Game/Sounds/Attenuation/Fort_Demolition_Hit_Small_Sized_Object_Att
 maxDistance=4000
 pitchMultiplier=1.7
 soundClassObject=&Fort_Impacts
 volumeMultiplier=0.68
 concurrencyOverrides {maxCount=3, resolutionRule=StopOldest}
 concurrencySet [&CONC_Object_Impacts_FourMax]
}
